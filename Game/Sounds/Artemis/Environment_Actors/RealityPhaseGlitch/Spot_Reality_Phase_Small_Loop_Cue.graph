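# Looping ambient cue for a small reality phase glitch spot sound, played quietly on the Athena ambient sound class with glitch attenuation.
# A looping wave feeds two stacked, endlessly looping Envelopers, one drifting pitch over sixty seconds, the other dipping and restoring volume over thirty.
# A Modulator at the root adds default random pitch and volume variation.
# A concurrency override caps it at two simultaneous instances.
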SoundCue /Game/Sounds/Artemis/Environment_Actors/RealityPhaseGlitch/Spot_Reality_Phase_Small_Loop_Cue
root=Modulator_0
nodes [
 {
  id=Enveloper_1
  children [WavePlayer_0]
  properties {
   bLoop=true
   bLoopIndefinitely=true
   loopEnd=60
   pitchCurve {editorCurveData {keys [
      {interpMode=RCIM_Cubic, time=0}
      <at 15.9468 0.9 RCIM_Cubic>
      {arriveTangent=0.0069, interpMode=RCIM_Cubic, leaveTangent=0.0069, time=30}
      <at 45 1.1>
      {time=60}
     ]}}
   volumeCurve {editorCurveData {keys [<at 0 1>]}}
  }
 }
 {
  id=Enveloper_2
  children [Enveloper_1]
  properties {
   bLoop=true
   bLoopIndefinitely=true
   loopEnd=30
   pitchCurve {editorCurveData {keys [{time=0}]}}
   volumeCurve {editorCurveData {keys [
      {arriveTangent=-0.01, interpMode=RCIM_Cubic, time=0, value=1}
      <at 15 0.6 RCIM_Cubic>
      {interpMode=RCIM_Cubic, leaveTangent=0.01, time=30, value=1}
     ]}}
  }
 }
 {id=Modulator_0, children [Enveloper_2]}
 {
  id=WavePlayer_0
  properties {bLooping=true, soundWaveAssetPtr=&/Spot_Reality_Phasing_Small_Loop_01}
 }
]
properties {
 attenuationSettings=&/Fort_Reality_Glitch_Att
 bOverrideConcurrency=true
 maxDistance=2250
 soundClassObject=&Fort_Ambient_Athena
 volumeMultiplier=0.301
 concurrencyOverrides {maxCount=2}
}
