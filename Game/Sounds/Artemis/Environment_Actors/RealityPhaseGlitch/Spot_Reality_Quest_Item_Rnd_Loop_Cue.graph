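# Looping spot ambience cue for a reality-glitch quest item, in the quests sound class with borrowed trash-flies attenuation settings.
# A quality-level node gives the first two levels the full mix and the other two only the shimmer loop.
# The mix layers that shimmer with quieter glitch one-shots, repeated after random delays with volume and pitch variation.
# The glitch picker weights six waves and an empty slot equally.
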SoundCue /Game/Sounds/Artemis/Environment_Actors/RealityPhaseGlitch/Spot_Reality_Quest_Item_Rnd_Loop_Cue
root=Modulator_0
nodes [
 {id=Delay_0, children [Modulator_1], properties {delayMax=2.5, delayMin=0.2}}
 {id=Looping_0, children [Delay_0]}
 {id=Mixer_0, children [Looping_0,WavePlayer_0], properties {inputVolume [0.175,1]}}
 {id=Modulator_0, children [QualityLevel_0], properties {pitchMax=1, volumeMax=1}}
 {id=Modulator_1, children [Random_0], properties {pitchMax=1.1, volumeMax=1, volumeMin=0.7}}
 {id=QualityLevel_0, children [Mixer_0,Mixer_0,WavePlayer_1,WavePlayer_1]}
 {
  id=Random_0
  children [WavePlayer_2,WavePlayer_3,WavePlayer_4,WavePlayer_5,WavePlayer_6,WavePlayer_7,null]
  properties {weights [1,1,1,1,1,1,1]}
 }
 {
  id=WavePlayer_0
  properties {bLooping=true, soundWaveAssetPtr=&/Spot_Reality_Quest_Item_Shimmer_Loop}
 }
 {
  id=WavePlayer_1
  properties {bLooping=true, soundWaveAssetPtr=&Spot_Reality_Quest_Item_Shimmer_Loop}
 }
 {id=WavePlayer_2, properties {soundWaveAssetPtr=&/Spot_Reality_Glitch_01}}
 {id=WavePlayer_3, properties {soundWaveAssetPtr=&/Spot_Reality_Glitch_02}}
 {id=WavePlayer_4, properties {soundWaveAssetPtr=&/Spot_Reality_Glitch_03}}
 {id=WavePlayer_5, properties {soundWaveAssetPtr=&/Spot_Reality_Glitch_04}}
 {id=WavePlayer_6, properties {soundWaveAssetPtr=&/Spot_Reality_Glitch_05}}
 {id=WavePlayer_7, properties {soundWaveAssetPtr=&/Spot_Reality_Glitch_06}}
]
properties {
 attenuationSettings=/Game/Athena/Sounds/Quests/CleaningUpTheMess/Spot_Flies_Around_Trash_Att
 maxDistance=2010
 soundClassObject=&Fort_Quests
 volumeMultiplier=1.5
 attenuationOverrides {falloffDistance=700, attenuationShapeExtents {x=100}}
}
